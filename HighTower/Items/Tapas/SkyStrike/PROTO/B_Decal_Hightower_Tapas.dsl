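; Decal Blueprint for the HighTower Tapas SkyStrike prototype that shows a crack decal fading in when it spawns.
; The construction script creates a dynamic instance of the Tapas crack decal material, stores it and assigns it to the decal component.
; On BeginPlay, a FadeIn timeline plays from the start and drives the material's FadeIn scalar parameter each update.
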(event EventBeginPlay
  (bind (_direction _newtrack_0) (Timeline|FadeIn PlayFromStart)
    (:Update
      (Rendering|Material|SetScalarParameterValue (Variables|Default|GetDecal_MID) "FadeIn" _newtrack_0 :!class /Script/Engine.MaterialInstanceDynamic))))

(fn ConstructionScript
  (bind _returnvalue (Rendering|Material|CreateDynamicMaterialInstance "/HighTower/Items/Tapas/SkyStrike/PROTO/MD_Hightower_Tapas_Crack.MD_Hightower_Tapas_Crack" "None" "None" :!class /Script/Engine.KismetMaterialLibrary))
  (bind _output_get (Variables|Default|SetDecal_MID _returnvalue))
  (Rendering|Components|Decal|SetDecalMaterial (Variables|Decal|GetDecal) _output_get))
