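# Prototype unlit, translucent deferred decal material for the Tapas SkyStrike item, showing an animated, glowing energy pattern within the Tapas decal shape.
# Caustics and electric noise textures pan at different speeds in radial coordinates, then multiply and square into a blend mask.
# That mask lerps from a near-black tint to intensely boosted orange, clipped by the decal texture's red channel.
# The same channel, decal lifetime opacity and a FadeIn parameter form the opacity term.
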Material /HighTower/Items/Tapas/SkyStrike/PROTO/M_Tapas_Decal
graph {
 properties {
  blendMode=BLEND_Translucent
  materialDomain=MD_DeferredDecal
  pixelDepthOffsetMode=PDOM_Legacy
  shadingModel=MSM_Unlit
  usedShadingModels=Unlit
 }
 nodes [
  Constant()
  Constant() {r=0.5}
  n1 = TextureSample() {texture=&/T_Decal_Tapas}
  n2 = VectorParameter() {defaultValue=<rgba 1 0.3436 0 1>}
  n3 = ScalarParameter() {defaultValue=450}
  n4 = MaterialFunctionCall(Vector or UVs=Subtract(A=TextureCoordinate()) {constB=0.5}) {materialFunction=&VectorToRadialValue}
  n5 = Multiply(
   A=TextureSample(UVs=Panner(Coordinate=@n4.Radial Coordinates) {speedX=0.01, speedY=-0.1}) {texture=/Game/Global/Textures/Default/Caustics/T_Caustics_Softer}.R
   B=TextureSample(UVs=Panner(Coordinate=@n4.Radial Coordinates) {speedX=0.006, speedY=-0.05}) {samplerType=SAMPLERTYPE_LinearColor, texture=/Game/Characters/Player/Male/Medium/Bodies/M_MED_SRDrift/Textures/T_ElectricNoise}.R
  )
  Multiply(
   A=@n1.R
   B=LinearInterpolate(
    A=VectorParameter() {parameterName=Param_1, defaultValue=<rgba 0.0156 0 0.0018 1>}.RGB
    B=Multiply(A=@n2.RGB, B=@n3)
    Alpha=Multiply(A=@n5, B=@n5)
   )
  )
  Multiply(A=Multiply(A=@n1.R, B=DecalLifetimeOpacity().Opacity), B=ScalarParameter() {defaultValue=1, parameterName=FadeIn})
 ]
}
